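# Translucent, unlit, two-sided flare material for Niagara mesh particles, filed under the DragonCart orb warning VFX.
# Panning noise distorts a cell-noise texture sampled by U alone for streaks, offset by a dynamic parameter, shaped by Refine and tinted by particle color.
# A V gradient, distorted smoke texture, inverse Fresnel and particle alpha form the fade mask, depth-faded in deferred shading.
# That mask also drives slight refraction, flattened on cheaper quality and shading paths.
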Material /DragonCartNarrative/Assets/VFX/Orb_Warning/M_DragonCart_Flare
graph {
 properties {
  bUsedWithNanite=true
  bUsedWithNiagaraMeshParticles=true
  blendMode=BLEND_Translucent
  pixelDepthOffsetMode=PDOM_Legacy
  refractionMethod=RM_IndexOfRefraction
  shadingModel=MSM_Unlit
  twoSided=true
  usedShadingModels=Unlit
 }
 nodes [
  Constant2Vector() {g=0.05, r=0.025}
  n1 = ScalarParameter() {parameterName=Distortion_U_Speed}
  n2 = ScalarParameter() {defaultValue=-0.2, parameterName=Distortion_V_Speed}
  n3 = TextureSample!(UVs=Panner(Coordinate=TextureCoordinate() {uTiling=2}, Speed=AppendVector(A=@n1, B=@n2)) {speedY=-0.2}, Tex, Level, Apply View MipBias) {mipValueMode=TMVM_MipLevel, samplerType=SAMPLERTYPE_Grayscale, texture=&LowResBlurredNoise}
  n4 = Multiply(
   A=@n3.R
   B=AppendVector(
    A=ScalarParameter() {defaultValue=0.025, parameterName=Distortion_U_Amp}
    B=ScalarParameter() {defaultValue=0.05, parameterName=Distortion_V_Amp}
   )
  )
  n5 = MaterialFunctionCall(Float2=TextureCoordinate()) {materialFunction=&BreakOutFloat2Components}
  n6 = ScalarParameter() {defaultValue=1, parameterName=Texture_U_Freq}
  n7 = ScalarParameter() {defaultValue=1, parameterName=Texture_V_Freq}
  n8 = DynamicParameter() {paramNames [OffsetX,OffsetY,Param3,Param4]}
  n9 = Add(A=Multiply(A=AppendVector(A=@n5.R, B=@n5.R), B=AppendVector(A=@n6, B=@n7)), B=AppendVector(A=@n8.OffsetX, B=@n8.OffsetX))
  n10 = ScalarParameter() {defaultValue=0.005, parameterName=OffsetChannel1}
  n11 = ScalarParameter() {defaultValue=0.02, parameterName=OffsetChannel2}
  n12 = TextureSampleParameter2D(UVs=Add(A=Add(A=@n4, B=@n9), B=LinearInterpolate(A=@n10, B=@n11, Alpha=@n3.R)) {constB=0}) {parameterName=Texture, samplerType=SAMPLERTYPE_Grayscale, texture=/Game/Weapons/FORT_Melee/Pickaxe_Cube_Ninja_Male/FX/T_CellNoise}
  n13 = VectorParameter() {parameterName=Refine, defaultValue=<rgba 1 1 1 1>}
  n14 = Add(A=Multiply(A=Power(Base=@n12.R, Exp=@n13.R) {constExponent=6}, B=@n13.G) {constB=100}, B=Multiply(A=@n12.R, B=@n13.B) {constB=5})
  n15 = Constant() {r=-0.5}
  n16 = ParticleColor()
  MaterialFunctionCall(Emissive=Multiply(A=Desaturation(None=@n14, Fraction=@n15), B=@n16.RGB)) {materialFunction=&MF_FX_EmissiveAdjustments}
  n17 = Multiply(A=Desaturation(None=@n12.R, Fraction=Constant() {r=1}), B=@n13.A) {constB=10}
  n18 = TextureCoordinate()
  n19 = Multiply(
   A=ComponentMask(None=@n18) {g=true}
   B=TextureSample!(UVs=Add(A=@n4, B=@n18), Tex, Apply View MipBias) {samplerType=SAMPLERTYPE_Grayscale, texture=/Game/Effects/Fort_Effects/Textures/Smoke/T_Smoke_Trail_Soft_NoClamp}.R
  )
  n20 = ScalarParameter() {defaultValue=5, parameterName=ExponentIn}
  n21 = Multiply(A=Multiply(A=@n16.A, B=Multiply(A=@n17, B=@n19)), B=OneMinus(None=Fresnel(ExponentIn=@n20)))
  n22 = DepthFade(Opacity=@n21, FadeDistance=ScalarParameter() {defaultValue=200, parameterName=DepthFade}) {fadeDistanceDefault=20}
  n23 = LinearInterpolate(Alpha=ShadingPathSwitch(Default=@n22, Deferred=@n22, Forward=@n21, Mobile=@n21)) {constA=1, constB=1.05}
  n24 = Constant() {r=1}
  n25 = QualitySwitch(Default=@n23, Low=@n24, High=@n23, Medium=@n24, Epic=@n23)
  ShadingPathSwitch(Default=@n25, Deferred=@n25, Forward=@n24, Mobile=@n24)
 ]
}
