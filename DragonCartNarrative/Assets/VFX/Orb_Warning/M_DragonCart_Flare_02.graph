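# Translucent, unlit, two-sided flare material for Niagara sprites in the DragonCart orb warning VFX.
# A center-scaled texture sample, a circle by default, is tinted by particle color, multiplied by a boost and passed through a shared emissive adjustment function.
# Opacity combines particle alpha, the texture's green channel and a boost, then applies camera-distance fade and depth fade to soften near-camera and intersecting sprites.
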Material /DragonCartNarrative/Assets/VFX/Orb_Warning/M_DragonCart_Flare_02
graph {
 properties {
  bAutomaticallySetUsageInEditor=false
  bUsedWithNiagaraSprites=true
  blendMode=BLEND_Translucent
  shadingModel=MSM_Unlit
  twoSided=true
  usedShadingModels=Unlit
 }
 nodes [
  n1 = ParticleColor()
  n2 = ScalarParameter() {defaultValue=1, group=Texture, parameterName=Texture Scale X}
  n3 = ScalarParameter() {defaultValue=1, group=Texture, parameterName=Texture Scale Y}
  n4 = TextureSampleParameter2D(UVs=MaterialFunctionCall(Texture Scale=AppendVector(A=@n2, B=@n3)) {materialFunction=&ScaleUVsByCenter}.UVs) {
   group=Texture
   parameterName=Flare_Texture
   samplerType=SAMPLERTYPE_LinearColor
   texture=&/T_DragonCart_Circle
  }
  n5 = Multiply(A=@n1.A, B=@n4.G)
  n6 = ScalarParameter() {defaultValue=1, parameterName=Opacity_Boost}
  n7 = ScalarParameter() {parameterName=Camera_FadeLength}
  n8 = ScalarParameter() {parameterName=Camera_FadeOffset}
  DepthFade(
   Opacity=Multiply(
    A=Multiply(A=@n5, B=@n6)
    B=MaterialFunctionCall(Fade Length=@n7, Fade Offset=@n8) {materialFunction=&CameraDepthFade}.Result
   )
   FadeDistance=ScalarParameter() {defaultValue=500, group=Opacity, parameterName=DepthFade}
  )
  MaterialFunctionCall(Emissive=Multiply(A=Multiply(A=@n1.RGB, B=@n4.RGB), B=ScalarParameter() {defaultValue=1, group=Emissive, parameterName=Emissive_Boost})) {materialFunction=&MF_FX_EmissiveAdjustments}
 ]
}
