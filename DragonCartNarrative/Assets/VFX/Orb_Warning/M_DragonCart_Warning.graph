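# Translucent, unlit, two-sided Niagara material for the DragonCart orb warning VFX, usable on sprite, ribbon and mesh particles.
# A procedural UV mask multiplies a U parabola, peaking at the centre and fading at both edges, by an inverted V gradient.
# Power parameters shape both falloffs.
# Particle alpha scales the mask, which feeds a depth fade with adjustable distance to soften geometry intersections.
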Material /DragonCartNarrative/Assets/VFX/Orb_Warning/M_DragonCart_Warning
graph {
 properties {
  bAutomaticallySetUsageInEditor=false
  bUsedWithNiagaraMeshParticles=true
  bUsedWithNiagaraRibbons=true
  bUsedWithNiagaraSprites=true
  blendMode=BLEND_Translucent
  shadingModel=MSM_Unlit
  twoSided=true
  usedShadingModels=Unlit
 }
 nodes [
  n1 = TextureCoordinate()
  n2 = ComponentMask(None=@n1) {r=true}
  n3 = OneMinus(None=@n2)
  n4 = Power(Base=Multiply(A=Multiply(A=@n3, B=@n2)) {constB=4}, Exp=ScalarParameter() {defaultValue=2, parameterName=EdgeFade_Power})
  n5 = Power(Base=OneMinus(None=ComponentMask(None=@n1) {g=true}), Exp=ScalarParameter() {defaultValue=2, parameterName=FadeEnding_Power})
  DepthFade(Opacity=Multiply(A=ParticleColor().A, B=Multiply(A=@n4, B=@n5)), FadeDistance=ScalarParameter() {defaultValue=200, parameterName=DepthFade})
 ]
}
