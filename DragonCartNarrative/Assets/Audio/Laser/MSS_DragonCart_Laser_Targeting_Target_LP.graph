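# Mono MetaSound source for the DragonCart orb laser's targeting loop at the target, using attenuation settings.
# Movement speed maps to a smoothed wave player pitch shift that falls from MaxPitch as speed rises.
# An enclosure modulation bus drives a one-pole low-pass and gain cut, followed by smoothed decibel volume randomized on play.
# Random pitch, azimuth and elevation speed, and sensitivity threshold inputs are unconnected.
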MetaSoundSource /DragonCartNarrative/Assets/Audio/Laser/MSS_DragonCart_Laser_Targeting_Target_LP
inputs [
 MaxPitch {type=Float}
 Movement.Speed {type=Float}
 Movement.SpeedAzimuth {type=Float}
 Movement.SpeedElevation {type=Float}
 Random.Pitch.Base {type=Float}
 Random.Pitch.Spread {type=Float}
 Random.Volume.Base {type=Float}
 Random.Volume.Spread {type=Float}
 SensitivityThreshold {type=Float}
 Sound {type=WaveAsset}
 UE.Source.OnPlay {type=Trigger}
]
interfaces [Fort.Movement:1.0, UE.OutputFormat.Mono:1.0, UE.Source.OneShot:1.0, UE.Source:1.1]
outputs [
 "UE.OutputFormat.Mono.Audio:0" {type=Audio}
 UE.Source.OneShot.OnFinished {type=Trigger}
]
pages [Default {
  input_defaults {
   MaxPitch=5
   Movement.Speed=254
   Movement.SpeedAzimuth=0
   Movement.SpeedElevation=0
   Random.Pitch.Base=5
   Random.Pitch.Spread=1
   Random.Volume.Base=1
   Random.Volume.Spread=0
   SensitivityThreshold=5
   Sound=/DragonCartNarrative/Assets/Audio/Laser/WAV/DragonCart_Orb_Laser_Targetting_Target_LP
   UE.Source.OnPlay=null
  }
  nodes [
   n1 = AbsoluteValue.Abs:Float(Input=^Random.Volume.Spread)
   DecibelstoLinearGain_0 = UE.Decibels to Linear Gain:Float(Decibels=UE.RandomFloat(
     Next=^UE.Source.OnPlay
     Max=UE.Add:Float(PrimaryOperand=^Random.Volume.Base, AdditionalOperands=@n1.Value).Out
     Min=UE.Subtract:Float(PrimaryOperand=^Random.Volume.Base, AdditionalOperands=@n1.Value).Out
    ).Value)
   n2 = Modulation.GetModulatorValue() {also {defaults {Modulator=&AudioWorldization_Bus_Enclosure}}}
   WavePlayer_0 = UE.Wave Player:Mono(
    Wave Asset=^Sound
    Play=^UE.Source.OnPlay
    Pitch Shift=UE.InterpTo:Audio(Target=MapRange.MapRange:Float(Out Range A=^MaxPitch, In=^Movement.Speed) {also {defaults {In Range B=400, Out Range B=0}}}.Out Value) {also {defaults {Interp Time=3}}}.Value
   ) {also {defaults {Loop="<MetaSoundBuilderResult.SUCCEEDED: 0>"}, version [1,1]}}
   OnePoleLowPassFilter_0 = UE.One-Pole Low Pass Filter:Audio(
    Cutoff Frequency=MapRange.MapRange:Float(In=@n2.Out) {also {defaults {In Range A=0.75, In Range B=0.9, Out Range A=20000, Out Range B=2500}}}.Out Value
    In=@WavePlayer_0.Out Mono
   )
   Multiply_1 = UE.Multiply:Audio by Float(
    AdditionalOperands=UE.InterpTo:Audio(Target=@DecibelstoLinearGain_0.Linear Gain) {also {defaults {Interp Time=0.5}}}.Value
    PrimaryOperand=UE.Multiply:Audio by Float(
     AdditionalOperands=MapRange.MapRange:Float(In=@n2.Out) {also {defaults {In Range A=0.75, In Range B=0.9, Out Range A=1, Out Range B=0.5}}}.Out Value
     PrimaryOperand=@OnePoleLowPassFilter_0.Out
    ).Out
   )
  ]
  links [{dst="@UE.OutputFormat.Mono.Audio:0", src=Multiply_1.Out}, {dst="@UE.Source.OneShot.OnFinished", src=WavePlayer_0.On Finished}]
 }]
properties {
 assetImportData=&self:AssetImportData
 attenuationSettings=&/ATT_DragonCart_Orb_Laser_Target
 maxDistance=83500
 sampleRate=48000
 soundClassObject=&Fort_Events_Gameplay
 volume=1.2
}
