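# Niagara warning effect for the DragonCart orb targeting the player, timed by User.TargetingDuration.
# A sprite beam with light and looping audio runs from a sampled static mesh point toward the system owner, wobbling and lagging less as TargetingLockCurve rises.
# A delayed mesh edge glow pulses based on that lock curve and fades when BehindObject is set, alongside a telegraph decal below the owner.
# A second edge glow emitter is disabled.
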NiagaraSystem /DragonCartNarrative/Assets/VFX/Orb_Warning/NS_DragonCart_Orb_PlayerWarning
dependencies {
 usedDataInterfaces [&NiagaraDataInterfaceCurve, &NiagaraDataInterfaceAudioPlayer, &NiagaraDataInterfaceArrayFloat]
 usedDynamicInputs [&self:ScratchDynamicInput_01]
 usedModules [
  &SystemState
  &self:SystemSpawnScript.SystemScriptSource.SystemScriptGraph.NiagaraNodeAssignment_3.SetVariables_#1
  &self:SystemSpawnScript.SystemScriptSource.SystemScriptGraph.NiagaraNodeAssignment_0.SetVariables_#2
  &Timeline
 ]
}
emitters {
 EdgeGlow {
  emitterUpdateScript [
   &EmitterState()
   &SpawnBurst_Instantaneous(Spawn Time=0.5, Use Spawn Probability=false, UseLoopCountLimit=false) {module=SpawnBurst_Instantaneous001}
  ]
  particleSpawnScript [
   &InitializeParticle(
    Color=<rgba 5 0 0 1>
    Lifetime=&Subtract_Float(A=$User.TargetingDuration, B=0.5)
    Mesh Scale=<6 6 6>
    Mesh Scale Mode=::NewEnumerator4
    Position=<0 0 -20>
    Position Mode=::NewEnumerator1
    UsePositionOffset=false
    Write Material Random=true
   )
   &InitialMeshOrientation(Mesh Axis To Orient=<0 0 1>, Mesh Orientation Mode=::NewEnumerator2, Use Rotation Vector=true)
  ]
  particleUpdateScript [
   &ParticleState()
   &CameraOffset(Camera Offset Amount=5)
   &ScaleColor(
    Scale RGB=&MakeCustomVectorFromBool(
     Bool=&SetBoolByFloatComparison(A=$System.TargetingLockCurve, B=1, Comparison Type=::NewEnumerator4)
     True Vector=&Multiply_VectorByFloat(
      Vector=<15 15 15>
      Float=&Waveform("Waveform [1]"=::NewEnumerator0, "[1] Frequency"=7, "[1] Amplitude Min/Max"=<0 1>)
     )
     False Vector=<8 8 8>
    )
    ScaleA=false
    ScaleRGB=true
   ) {module=ScaleColor001}
   &ScaleColor(
    Scale Alpha=&MakeCustomFloatFromBool(
     Bool=$User.BehindObject
     True Float=&FloatFromCurve(FloatCurve=DataInterface() {propertyValues {Curve {keys [{interpMode=RCIM_Cubic, tangentMode=RCTM_Break, time=0, value=1}, {interpMode=RCIM_Cubic, tangentMode=RCTM_Break, time=1, value=0}]}}})
     False Float=1
    )
    ScaleA=true
    ScaleRGB=true
   )
   &DynamicMaterialParameters(Module.Index0Param0Write=true, Module.Index0Param1Write=true, Module.Index0Param2Write=true, Module.Index0Param3Write=true)
  ]
  rendererClasses [&NiagaraMeshRendererProperties]
  renderers [{rendererClass=&NiagaraMeshRendererProperties, rendererIndex=0}]
 }
 EdgeGlow001 {
  bEnabled=false
  emitterUpdateScript [
   &EmitterState()
   &SpawnBurst_Instantaneous(Spawn Time=0.5, Use Spawn Probability=false, UseLoopCountLimit=false) {module=SpawnBurst_Instantaneous001}
  ]
  particleSpawnScript [
   &InitializeParticle(
    Color=<rgba 5 0 0 1>
    Lifetime=&Subtract_Float(A=$User.TargetingDuration, B=0.5)
    Mesh Scale=<2 2 3>
    Mesh Scale Mode=::NewEnumerator4
    UsePositionOffset=false
    Write Material Random=true
   )
   &InitialMeshOrientation(Mesh Axis To Orient=<0 0 1>, Mesh Orientation Mode=::NewEnumerator2, Use Rotation Vector=true)
  ]
  particleUpdateScript [
   &ParticleState()
   &CameraOffset(Camera Offset Amount=5)
   &ScaleColor(
    Scale RGB=&MakeCustomVectorFromBool(
     Bool=&SetBoolByFloatComparison(A=$System.TargetingLockCurve, B=1, Comparison Type=::NewEnumerator4)
     True Vector=&Multiply_VectorByFloat(
      Vector=<15 15 15>
      Float=&Waveform("Waveform [1]"=::NewEnumerator0, "[1] Frequency"=7, "[1] Amplitude Min/Max"=<0 1>)
     )
     False Vector=<8 8 8>
    )
    ScaleA=false
    ScaleRGB=true
   ) {module=ScaleColor001}
   &ScaleColor(
    Scale Alpha=&MakeCustomFloatFromBool(
     Bool=$User.BehindObject
     True Float=&FloatFromCurve(FloatCurve=DataInterface() {propertyValues {Curve {keys [{interpMode=RCIM_Cubic, tangentMode=RCTM_Break, time=0, value=1}, {interpMode=RCIM_Cubic, tangentMode=RCTM_Break, time=1, value=0}]}}})
     False Float=1
    )
    ScaleA=true
    ScaleRGB=true
   )
   &DynamicMaterialParameters(Module.Index0Param0Write=true, Module.Index0Param1Write=true, Module.Index0Param2Write=true, Module.Index0Param3Write=true)
  ]
  rendererClasses [&NiagaraMeshRendererProperties]
  renderers [{rendererClass=&NiagaraMeshRendererProperties, rendererIndex=0}]
 }
 TargetingBeam001 {
  emitterSpawnScript [&InitializeAudioPlayer(Audio Player=DataInterface() {propertyValues {
      Attenuation=null
      Concurrency=null
      MaxPlaysPerTick=10
      SoundToPlay=/DragonCartNarrative/Assets/Audio/Laser/MSS_DragonCart_Laser_Targeting_Target_LP
      bAllowLoopingOneShotSounds=false
      bLimitPlaysPerTick=true
      bOnlyActiveDuringGameplay=false
      bStopWhenComponentIsDestroyed=true
      ConfigurationUserParameter {parameter {name=None, typeDefHandle {registeredTypeIndex=1}, varData []}}
      ParameterNames []
     }})]
  emitterUpdateScript [
   &EmitterState()
   &SpawnBurst_Instantaneous(Use Spawn Probability=false, UseLoopCountLimit=false)
  ]
  particleSpawnScript [
   &InitializeParticle(
    Color=$User.Primary_Color
    Lifetime=$User.TargetingDuration
    Mesh Scale Mode=::NewEnumerator3
    Mesh Uniform Scale=0.6
    Sprite Size Mode=::NewEnumerator3
    Uniform Sprite Size=2200
    UsePositionOffset=false
    Write Material Random=true
   )
   /SetVariables(Particles.LightExponent=2) {isSetParameters=true}
   /SetVariables(Particles.LightRadius=150) {isSetParameters=true}
   /SetVariables(Particles.LightColor=$User.Light_Color) {isSetParameters=true}
   &DynamicMaterialParameters(
    Index 0 Param 1=2
    Index 0 Param 2=10
    Index 0 Param 3=1
    Index 0 Param 4=1
    Index 3 Param 1=10
    Module.Index0Float0Write=true
    Module.Index0Float1Write=true
    Module.Index0Float2Write=true
    Module.Index0Float3Write=true
    Module.Index3Float0Write=true
    Module.Index3Float1Write=true
    Module.Index3Float2Write=true
    Module.Index3Float3Write=true
    Write Parameter Index 3=::NewEnumerator1
   )
   &CameraOffset(Camera Offset Amount=100)
   /Niagara/Modules/Audio/PlayPersistentAudio()
  ]
  particleUpdateScript [
   &ParticleState()
   /Niagara/Modules/Spawn/Location/StaticMeshLocation(
    MaskSpawnGroup=false
    Mesh Sampling Type:ENiagara_StaticSamplingMode=::NewEnumerator1
    SamplePosOffset=true
    Sampled Position Offset:Vector3f=<10 0 0>
    Socket Sampling Mode:ENiagara_StaticSocketMode=::NewEnumerator2
    Static Mesh:NiagaraDataInterfaceStaticMesh=$User.Static Mesh Reader
   )
   /SetVariables(Particles.Position=&AddVectorToPosition(
     Position=$Particles.Position
     Vector=&TransformVector(Vector=$System.OrbWPO, Source Space=::World, Destination Space=::Local)
    )) {isSetParameters=true}
   /SetVariables(Particles.EndPointStartPosition:NiagaraPosition=&AddVectorToPosition(
     Position=$Particles.StaticMeshLocation.SampledPosition
     Vector=&Multiply_VectorByFloat(
      Vector=&ReturnMeshOrientationAxis(Mesh Orientation Axis=::XAxis, Mesh Orientation Quat=$Particles.MeshOrientation)
      Float=&FloatFromCurve(
       FloatCurve=DataInterface() {propertyValues {Curve {defaultValue=1, keys [{tangentMode=RCTM_User, time=0, value=0}, {tangentMode=RCTM_User, time=0.1, value=1}]}}}
       CurveIndex=$Particles.Age
       Scale Curve=8000
      )
     )
    )) {isSetParameters=true}
   /SetVariables(Particles.Oscillation:Vector3f=&Multiply_VectorByFloat(
     Vector=&MakeVector(
      X=&Waveform("Waveform [1]"=::NewEnumerator0, "[1] Frequency"=0.5, "[1] Amplitude Min/Max"=<-1 1>)
      Y=&Waveform("Waveform [1]"=::NewEnumerator0, "[1] Frequency"=0.6, "[1] Amplitude Min/Max"=<-1 1>)
      Z=&Waveform("Waveform [1]"=::NewEnumerator0, "[1] Frequency"=0.8, "[1] Amplitude Min/Max"=<-1 1>)
     )
     Float=&Lerp_Float(A=100, B=0, Alpha=$System.TargetingLockCurve)
    )) {isSetParameters=true}
   &InterpolateOverTime(
    Rate Of Change:NiagaraFloat=&Lerp_Float(A=4, B=50, Alpha=$System.TargetingLockCurve)
    Target Position Value:NiagaraPosition=&AddVectorToPosition(Position=$Engine.Owner.Position, Vector=$Particles.Oscillation)
    Variable Type=::NewEnumerator6
   )
   /SetVariables(Particles.EndPointPosition:NiagaraPosition=&Lerp_Vector(
     A=Linked() {linkedVariable {name=Particles.EndPointStartPosition, type=NiagaraPosition}}
     B=Linked() {linkedVariable {name=Particles.InterpolateOverTime.InterpolateOverTime_Position.Moving Average, type=NiagaraPosition}}
     Alpha=&FloatFromCurve(
      FloatCurve=DataInterface() {propertyValues {Curve {defaultValue=1, keys [<at 0.1 0 RCIM_Cubic>, {tangentMode=RCTM_User, time=0.5, value=1}]}}}
      CurveIndex=$Particles.Age
     )
    )) {isSetParameters=true}
   /SetVariables(Particles.BeamPosition:NiagaraPosition=&Lerp_Vector(
     A=Linked() {linkedVariable {name=Particles.Position, type=NiagaraPosition}}
     B=Linked() {linkedVariable {name=Particles.EndPointPosition, type=NiagaraPosition}}
     Alpha=0.5
    )) {isSetParameters=true}
   /SetVariables(Particles.BeamSpriteSize:Vector2f=&MakeVector2D(
     X=&FN_LOD_Distance_Remap(Start Distance=2000, End Distance=10000, Minimum Output=10, Maximum Output=50, Clamp Output=true)
     Y=&Multiply_Float(A=&DistanceBetweenPositions(A=$Particles.EndPointPosition, B=$Particles.Position), B=1)
    )) {isSetParameters=true}
   /SetVariables(Particles.BeamSpriteSizeWIder:Vector2f=&MakeVector2D(
     X=&FN_LOD_Distance_Remap(Start Distance=2000, End Distance=10000, Minimum Output=120, Maximum Output=420, Clamp Output=true)
     Y=&Multiply_Float(A=&DistanceBetweenPositions(A=$Particles.EndPointPosition, B=$Particles.Position), B=1)
    )) {isSetParameters=true}
   /SetVariables(Particles.EndPointSpriteSize:Vector2f=&Vector2DFromFloat(Value=&Lerp_Float(
      A=80
      B=120
      Alpha=&FloatFromCurve(
       FloatCurve=DataInterface() {propertyValues {Curve {keys [{interpMode=RCIM_Cubic, tangentMode=RCTM_Break, time=0.5, value=0}, {interpMode=RCIM_Cubic, tangentMode=RCTM_Break, time=1, value=1}]}}}
       CurveIndex=$System.NormalizedLoopAge
      )
     ))) {isSetParameters=true}
   /SetVariables(Particles.EndPointCircleSize:Vector2f=&Vector2DFromFloat(Value=&Lerp_Float(A=70, B=5, Alpha=$System.TargetingLockCurve))) {isSetParameters=true}
   /SetVariables(Particles.WiderBeam_Color:LinearColor=$User.Secondary_Color) {isSetParameters=true}
   &SpriteFacingAndAlignment(
    Sprite Alignment=&Subtract_Position(A=$Particles.Position, B=$Particles.EndPointPosition)
    Sprite Facing=&ReturnMeshOrientationAxis(Mesh Orientation Axis=::XAxis, Mesh Orientation Quat=$Particles.MeshOrientation)
    WriteAlignment=true
    WriteFacing=true
   )
   &ScaleMeshSize(Scale Factor=&VectorFromFloat(Value=&Lerp_Float(A=1, B=0.1, Alpha=$System.TargetingLockCurve)))
   &DynamicMaterialParameters(
    Index 0 Param 1=1
    Index 0 Param 2=100
    Index 0 Param 3=1
    Module.Index0Float0Write=true
    Module.Index0Float1Write=true
    Module.Index0Float2Write=true
    Module.Index0Float3Write=true
   ) {module=DynamicMaterialParameters001}
   /SetVariables(Particles.Sprite Start Camera Offset=300) {isSetParameters=true}
   /Niagara/Modules/Audio/UpdatePersistentAudio(Position:NiagaraPosition=$Particles.EndPointPosition, Update Type:ENiagara_AudioParamType=::NewEnumerator5)
  ]
  rendererClasses [&NiagaraSpriteRendererProperties, &NiagaraLightRendererProperties]
  renderers [
   {rendererClass=&NiagaraSpriteRendererProperties}
   {rendererClass=&NiagaraSpriteRendererProperties, rendererIndex=1}
   {rendererClass=&NiagaraSpriteRendererProperties, rendererIndex=2}
   {rendererClass=&NiagaraLightRendererProperties, rendererIndex=3}
   {rendererClass=&NiagaraSpriteRendererProperties, rendererIndex=4}
   {rendererClass=&NiagaraSpriteRendererProperties, rendererIndex=5}
  ]
 }
 TelegraphDecal001 {
  emitterUpdateScript [
   &EmitterState()
   &SpawnBurst_Instantaneous(Spawn Time=0.5, Use Spawn Probability=false, UseLoopCountLimit=false)
   /SetVariables(Emitter.OverallOpacity:NiagaraFloat=&FloatFromCurve(
     FloatCurve=DataInterface() {propertyValues {Curve {keys [
         {interpMode=RCIM_Cubic, tangentMode=RCTM_Break, time=0, value=0}
         {arriveTangent=-0.144, interpMode=RCIM_Cubic, leaveTangent=-0.144, tangentMode=RCTM_User, time=0.5, value=1}
         {arriveTangent=-0.1699, interpMode=RCIM_Cubic, leaveTangent=-0.1699, tangentMode=RCTM_User, time=0.7, value=1}
         {interpMode=RCIM_Cubic, tangentMode=RCTM_Break, time=1, value=0}
        ]}}}
     CurveIndex=$Emitter.NormalizedLoopAge
    )) {isSetParameters=true}
  ]
  particleSpawnScript [
   &InitializeParticle(
    Lifetime=&Subtract_Float(A=$User.TargetingDuration, B=0.5)
    Position Mode=::NewEnumerator1
    Position Offset=<0 0 -145>
    Sprite Size Mode=::NewEnumerator3
    Uniform Sprite Size=600
    UsePositionOffset=true
    Write Material Random=true
   )
   /SetVariables(
    Particles.DecalOrientation:Quat4f=&OrientMeshToVector(Look At Direction=<0 0 1>, Reference Axis Vector=<1 0 0>)
    Particles.DecalSize:Vector3f=<100 600 600>
   ) {isSetParameters=true}
  ]
  particleUpdateScript [
   &ParticleState()
   /SetVariables(Particles.DecalFade=0.2) {isSetParameters=true}
  ]
  rendererClasses [&NiagaraDecalRendererProperties]
  renderers [{rendererClass=&NiagaraDecalRendererProperties, rendererIndex=0}]
 }
}
summary {
 systemName=NS_DragonCart_Orb_PlayerWarning
 emitters [
  {emitterName=TelegraphDecal001, rendererClasses [&NiagaraDecalRendererProperties]}
  {bEnabled=false, emitterName=EdgeGlow001, rendererClasses [&NiagaraMeshRendererProperties]}
  {
   emitterName=TargetingBeam001
   rendererClasses [&NiagaraSpriteRendererProperties, &NiagaraLightRendererProperties]
  }
  {emitterName=EdgeGlow, rendererClasses [&NiagaraMeshRendererProperties]}
 ]
 userVariables [
  User.BehindObject {type=NiagaraBool}
  User.BuildUpIntensity {type=NiagaraFloat, defaultValue=1, description=""}
  User.Light_Color {type=LinearColor, description="", defaultValue {struct=LinearColor, value=<rgba 15 0 0 1>}}
  User.Primary_Color {type=LinearColor, description="", defaultValue {struct=LinearColor, value=<rgba 2.5 0 0 1>}}
  User.Secondary_Color {type=LinearColor, description="", defaultValue {struct=LinearColor, value=<rgba 1 0.112 0 0.2>}}
  User.Static Mesh Reader {
   type=NiagaraDataInterfaceStaticMesh
   description=""
   defaultValue {
    struct=NiagaraExt_VariableValue_DataInterface
    value {
     dataInterface=&self:NiagaraDataInterfaceStaticMesh_0
     dataInterfaceClass=&NiagaraDataInterfaceStaticMesh
    }
   }
  }
  User.TargetingDuration {type=NiagaraFloat, defaultValue=5, description=""}
 ]
}
