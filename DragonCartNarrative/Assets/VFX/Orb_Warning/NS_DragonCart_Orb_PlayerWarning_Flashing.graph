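# Niagara warning effect for the DragonCart orb, showing the player a flashing red targeting beam with a charge-up at the orb.
# All four emitters track a sampled static-mesh location plus the orb's world-position offset.
# TargetingBeam001 sweeps its endpoint from User.InitialPosition to the owner, pulses via a waveform, and widens with distance.
# Glow, plasma and lightning layers animate over particle or loop age, dimming by time of day.
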NiagaraSystem /DragonCartNarrative/Assets/VFX/Orb_Warning/NS_DragonCart_Orb_PlayerWarning_Flashing
dependencies {
 usedDataInterfaces [&NiagaraDataInterfaceCurve, &NiagaraDataInterfaceArrayFloat]
 usedDynamicInputs [&self:ScratchDynamicInput_01]
 usedModules [
  &self:SystemSpawnScript.SystemScriptSource.SystemScriptGraph.NiagaraNodeAssignment_3.SetVariables_#1
  &SystemState
  &self:SystemSpawnScript.SystemScriptSource.SystemScriptGraph.NiagaraNodeAssignment_0.SetVariables_#2
  &Timeline
 ]
}
emitters {
 Glow_02 {
  emitterUpdateScript [
   &EmitterState()
   &SpawnBurst_Instantaneous(Use Spawn Probability=false, UseLoopCountLimit=false)
  ]
  particleSpawnScript [
   &InitializeParticle(
    Color=<rgba 1 0 0 1>
    Lifetime=$User.Duration
    Position Mode=::NewEnumerator0
    Sprite Rotation Mode=::NewEnumerator2
    Sprite Size Mode=::NewEnumerator3
    Uniform Sprite Size=2600
    Write Material Random=true
   )
   &CameraOffset(Camera Offset Amount=300)
  ]
  particleUpdateScript [
   &ParticleState()
   /Niagara/Modules/Spawn/Location/StaticMeshLocation(
    MaskSpawnGroup=false
    Mesh Sampling Type:ENiagara_StaticSamplingMode=::NewEnumerator1
    SamplePosOffset=true
    Sampled Position Offset:Vector3f=<10 0 0>
    Socket Sampling Mode:ENiagara_StaticSocketMode=::NewEnumerator2
    Static Mesh:NiagaraDataInterfaceStaticMesh=$User.Static Mesh Reader
   )
   /SetVariables(Particles.Position=&AddVectorToPosition(
     Position=$Particles.Position
     Vector=&TransformVector(Vector=$System.OrbWPO, Source Space=::World, Destination Space=::Local)
    )) {isSetParameters=true}
   &ScaleSpriteSize(Uniform Curve Sprite Scale=DataInterface() {propertyValues {Curve {keys [
        {interpMode=RCIM_Cubic, tangentMode=RCTM_User, time=0, value=0.5}
        {
         arriveTangent=0.9101
         arriveTangentWeight=0.1572
         interpMode=RCIM_Cubic
         leaveTangent=0.9101
         leaveTangentWeight=0.187
         tangentMode=RCTM_User
         time=0.2013
         value=0.7311
        }
        {arriveTangentWeight=0.0486, interpMode=RCIM_Cubic, leaveTangentWeight=0.0063, tangentMode=RCTM_User, time=0.2316, value=0.6814}
        {arriveTangentWeight=0.0022, interpMode=RCIM_Cubic, leaveTangentWeight=0.0079, tangentMode=RCTM_User, time=0.2582, value=0.8592}
        {arriveTangentWeight=0.0079, interpMode=RCIM_Cubic, leaveTangentWeight=0.005, tangentMode=RCTM_User, time=0.2794, value=0.7085}
        {arriveTangentWeight=0.0056, interpMode=RCIM_Cubic, leaveTangentWeight=0.0103, tangentMode=RCTM_User, time=0.2947, value=1.0562}
        {arriveTangentWeight=0.0098, interpMode=RCIM_Cubic, leaveTangentWeight=0.0053, tangentMode=RCTM_User, time=0.3173, value=0.8333}
        {arriveTangentWeight=0.0153, interpMode=RCIM_Cubic, leaveTangentWeight=0.0391, tangentMode=RCTM_User, time=0.3257, value=1.2148}
        {arriveTangentWeight=0.2165, interpMode=RCIM_Cubic, leaveTangentWeight=0.1618, tangentMode=RCTM_User, time=0.3479, value=1.328}
        {arriveTangentWeight=0.1546, interpMode=RCIM_Cubic, leaveTangentWeight=0.0794, tangentMode=RCTM_User, time=0.3599, value=0.9473}
        {arriveTangentWeight=0.0037, interpMode=RCIM_Cubic, leaveTangentWeight=0.003, tangentMode=RCTM_User, time=0.3772, value=1.2876}
        {arriveTangentWeight=0.003, interpMode=RCIM_Cubic, leaveTangentWeight=0.0044, tangentMode=RCTM_User, time=0.3892, value=0.9119}
        {arriveTangentWeight=0.0044, interpMode=RCIM_Cubic, leaveTangentWeight=0.0039, tangentMode=RCTM_User, time=0.3981, value=1.849}
        {arriveTangentWeight=0.0039, interpMode=RCIM_Cubic, leaveTangentWeight=0.0028, tangentMode=RCTM_User, time=0.4083, value=1.285}
        {arriveTangentWeight=0.0028, interpMode=RCIM_Cubic, leaveTangentWeight=0.0048, tangentMode=RCTM_User, time=0.4181, value=1.7376}
        {arriveTangentWeight=0.0048, interpMode=RCIM_Cubic, leaveTangentWeight=0.0024, tangentMode=RCTM_User, time=0.4338, value=1.4407}
        {arriveTangentWeight=0.146, interpMode=RCIM_Cubic, leaveTangentWeight=0.1771, tangentMode=RCTM_User, time=0.4486, value=1.9071}
        {arriveTangentWeight=0.0029, interpMode=RCIM_Cubic, leaveTangentWeight=0.0043, tangentMode=RCTM_User, time=0.4668, value=1.3099}
        {arriveTangentWeight=0.014, interpMode=RCIM_Cubic, leaveTangentWeight=0.0241, tangentMode=RCTM_User, time=0.4777, value=1.9292}
        {arriveTangentWeight=0.0505, interpMode=RCIM_Cubic, leaveTangentWeight=0.0235, tangentMode=RCTM_User, time=0.5033, value=2.2433}
        {arriveTangentWeight=0.0186, interpMode=RCIM_Cubic, leaveTangentWeight=0.0293, tangentMode=RCTM_User, time=0.5163, value=0.896}
        {arriveTangentWeight=0.0007, interpMode=RCIM_Cubic, leaveTangentWeight=0.1139, tangentMode=RCTM_User, time=0.525, value=2.2846}
        {arriveTangentWeight=0.0104, interpMode=RCIM_Cubic, leaveTangentWeight=0.1242, tangentMode=RCTM_User, time=0.5633, value=1.7}
        {arriveTangent=0.5015, interpMode=RCIM_Cubic, leaveTangent=0.5015, time=0.6173, value=1.6127}
        {arriveTangent=-0.0268, interpMode=RCIM_Cubic, leaveTangent=-0.0268, time=0.7032, value=1.7701}
        {arriveTangent=-0.3546, interpMode=RCIM_Cubic, leaveTangent=-0.3546, time=0.8235, value=1.6072}
        {interpMode=RCIM_Cubic, tangentMode=RCTM_User, time=0.901, value=1.7}
        {arriveTangentWeight=0.0737, interpMode=RCIM_Cubic, leaveTangentWeight=0.0932, tangentMode=RCTM_User, time=0.92, value=2.2798}
        {arriveTangentWeight=0.6437, interpMode=RCIM_Cubic, tangentMode=RCTM_User, time=0.935, value=0}
       ]}}})
   &ScaleColor(
    Scale Alpha=&FloatFromCurve(
     FloatCurve=DataInterface() {propertyValues {Curve {keys [
         {arriveTangentWeight=0.0865, interpMode=RCIM_Cubic, leaveTangentWeight=0.2862, tangentMode=RCTM_User, time=0.1, value=1}
         {arriveTangentWeight=0.2434, interpMode=RCIM_Cubic, leaveTangentWeight=0.1554, tangentMode=RCTM_User, time=0.8, value=1}
         <at 0.9 0 RCIM_Cubic>
        ]}}}
     Scale Curve=&FNTimeOfDayScalar(Night Scale=0.5, Morning Scale=1, Evening Scale=1)
    )
    Scale RGB=&VectorFromFloat(Value=&FloatFromCurve(FloatCurve=DataInterface() {propertyValues {Curve {keys [
          {arriveTangent=-0.0616, interpMode=RCIM_Cubic, leaveTangent=-0.0616, tangentMode=RCTM_User, time=0.1, value=1}
          {arriveTangent=-0.044, interpMode=RCIM_Cubic, leaveTangent=-0.044, tangentMode=RCTM_User, time=0.8, value=1}
          {arriveTangent=-0.1284, interpMode=RCIM_Cubic, leaveTangent=-0.1284, tangentMode=RCTM_User, time=1, value=0.5}
         ]}}}))
    ScaleA=true
    ScaleRGB=true
   )
  ]
  rendererClasses [&NiagaraSpriteRendererProperties]
  renderers [{rendererClass=&NiagaraSpriteRendererProperties}]
 }
 Lightning {
  emitterUpdateScript [
   &EmitterState()
   &SpawnRate(
    SpawnRate=&FloatFromCurve(
     FloatCurve=DataInterface() {propertyValues {Curve {keys [
         {arriveTangent=2.6388, interpMode=RCIM_Cubic, tangentMode=RCTM_Break, time=0, value=1}
         {arriveTangent=-0.2281, interpMode=RCIM_Cubic, leaveTangent=-0.2281, tangentMode=RCTM_User, time=0.1, value=3}
         {arriveTangent=-0.0057, interpMode=RCIM_Cubic, leaveTangent=-0.0057, tangentMode=RCTM_User, time=0.3368, value=3.3}
         <at 1 0 RCIM_Constant>
        ]}}}
     CurveIndex=$System.NormalizedLoopAge
     Scale Curve=25
    )
    Use Spawn Probability=false
   )
  ]
  particleSpawnScript [
   /SetVariables(Particles.BiasedRandom=&FloatFromCurve(
     FloatCurve=DataInterface() {propertyValues {Curve {keys [
         {interpMode=RCIM_Cubic, tangentMode=RCTM_User, time=0, value=0}
         {arriveTangent=2.5197, interpMode=RCIM_Cubic, leaveTangent=2.5197, tangentMode=RCTM_User, time=0.6, value=0.6}
         {interpMode=RCIM_Cubic, tangentMode=RCTM_User, time=1, value=1}
        ]}}}
     CurveIndex=&RandomRangeFloat(Minimum=0, Maximum=1)
    )) {isSetParameters=true}
   /Niagara/Modules/Spawn/Initialization/InitializeParticle(
    Color=<rgba 1 0 0 0.15>
    Lifetime=&FloatFromCurve(
     FloatCurve=DataInterface() {propertyValues {Curve {keys [{interpMode=RCIM_Cubic, tangentMode=RCTM_Break, time=0, value=0.45}, {interpMode=RCIM_Cubic, tangentMode=RCTM_Break, time=0.3, value=0.2}]}}}
     CurveIndex=$System.NormalizedLoopAge
     Scale Curve=&RandomRangeFloat(Minimum=0.6, Maximum=1)
    )
    Mesh Scale=&MakeVector(
     X=&RandomRangeFloat(Minimum=1, Maximum=2)
     Y=&RandomRangeFloat(Minimum=0.25, Maximum=2)
     Z=&RandomRangeFloat(Minimum=0.25, Maximum=2)
    )
    Module.Write Color=true
    Module.Write Lifetime=true
    Module.Write Mass=true
    Module.Write Material Random=true
    Module.Write Position=false
    Module.Write Scale=true
    Module.Write SpriteRotation=false
    Module.Write SpriteSize=true
    Position=&SimulationPosition()
    Sprite Size=&Vector2DFromFloat(Value=&RandomRangeFloat(Minimum=300, Maximum=700))
   )
   &InitialMeshOrientation(Use Rotation Vector=true) {module=InitialMeshOrientation001}
   /SetVariables(Particles.MeshIndex=&RandomRangeInt(Minimum=0, Maximum=1)) {isSetParameters=true}
   /SetVariables(Particles.DynParamsSprite:Vector4f=<0.25 2 1 1>) {isSetParameters=true}
   /SetVariables(Particles.CamOffsetSprite=50) {isSetParameters=true}
   &ShapeLocation(
    Cylinder Distribution=::NewEnumerator2
    Cylinder Height=10
    Cylinder Radius=&RandomRangeFloat(Minimum=900, Maximum=1000)
    MaskSpawnGroup=false
    Shape Primitive=::NewEnumerator1
    U Position=&RandomRangeFloat(Minimum=0, Maximum=1)
    UseOwnerScale=true
   )
   &CameraOffset(Camera Offset Amount=200)
  ]
  particleUpdateScript [
   &ParticleState()
   &StaticMeshLocation(
    MaskSpawnGroup=false
    Mesh Sampling Type:ENiagara_StaticSamplingMode=::NewEnumerator1
    SamplePosOffset=true
    Sampled Position Offset:Vector3f=<10 0 0>
    Socket Sampling Mode:ENiagara_StaticSocketMode=::NewEnumerator2
    Static Mesh:NiagaraDataInterfaceStaticMesh=$User.Static Mesh Reader
   )
   /SetVariables(Particles.Position=&AddVectorToPosition(
     Position=$Particles.Position
     Vector=&TransformVector(Vector=$System.OrbWPO, Source Space=::World, Destination Space=::Local)
    )) {isSetParameters=true}
   &ScaleColor(
    Scale Alpha=&FloatFromCurve(FloatCurve=DataInterface() {propertyValues {Curve {keys [
         <at 0 0>
         {interpMode=RCIM_Cubic, tangentMode=RCTM_User, time=0.1258, value=0.1952}
         <at 0.341 0.0902>
         {interpMode=RCIM_Cubic, tangentMode=RCTM_User, time=0.496, value=0.2532}
         <at 1 0>
        ]}}})
    Scale RGB=&VectorFromFloat(Value=&FloatFromCurve(FloatCurve=DataInterface() {propertyValues {Curve {keys [
          {arriveTangent=25.5746, interpMode=RCIM_Cubic, leaveTangent=25.5746, tangentMode=RCTM_User, time=0, value=0.5}
          {arriveTangent=-13.4076, interpMode=RCIM_Cubic, leaveTangent=-13.4076, tangentMode=RCTM_User, time=0.3, value=5}
          <at 1 1>
         ]}}}))
    ScaleA=true
    ScaleRGB=true
   )
   &ScaleColor(
    Scale Alpha=0
    Scale RGB=&VectorFromFloat(Value=&FNTimeOfDayScalar(Night Scale=0.4, Morning Scale=1, Evening Scale=1))
    ScaleA=true
    ScaleRGB=true
   ) {module=ScaleColor001}
   /SetVariables(Particles.SpriteColor=&MultiplyLinearColor(
     Linear Color=$Particles.Color
     Scale Factor=&MakeVector4_FromLinearColor(LinearColor=&MultiplyLinearColor(
       Linear Color=<rgba 25 15.0635 2.969 1>
       Scale Factor=&MakeVector4(
        X=1
        Y=1
        Z=1
        W=&FloatFromCurve(
         FloatCurve=DataInterface() {propertyValues {Curve {keys [
             {interpMode=RCIM_Cubic, tangentMode=RCTM_User, time=0, value=1}
             {arriveTangentWeight=0.031, interpMode=RCIM_Cubic, leaveTangentWeight=0.0366, tangentMode=RCTM_User, time=0.0918, value=0.05}
             {arriveTangentWeight=0.16, interpMode=RCIM_Cubic, leaveTangent=18.7153, leaveTangentWeight=0.1479, tangentMode=RCTM_Break, time=0.18, value=0}
             {arriveTangent=-0.2815, interpMode=RCIM_Cubic, leaveTangent=-0.2815, tangentMode=RCTM_User, time=0.3, value=0.95}
             {arriveTangentWeight=0.0345, interpMode=RCIM_Cubic, leaveTangentWeight=0.0566, tangentMode=RCTM_User, time=0.4, value=0.1}
             {
              arriveTangentWeight=0.1589
              interpMode=RCIM_Cubic
              leaveTangent=29.6956
              leaveTangentWeight=0.1018
              tangentMode=RCTM_Break
              time=0.5
              value=0.05
             }
             {arriveTangent=0.512, interpMode=RCIM_Cubic, leaveTangent=0.512, tangentMode=RCTM_User, time=0.5641, value=0.8648}
             <at 1 0>
            ]}}}
         Scale Curve=&Multiply_Float(
          A=$System.NearFade
          B=&FNTimeOfDayScalar(Night Scale=0.5, Morning Scale=0.7, Day Scale=0.8, Evening Scale=0.6)
         )
        )
       )
      ))
    )) {enabled=false, isSetParameters=true}
   &DynamicMaterialParameters(
    Index 0 Param 1=&FloatFromCurve(FloatCurve=DataInterface() {propertyValues {Curve {defaultValue=1, keys [{tangentMode=RCTM_User, time=0, value=0}, {tangentMode=RCTM_User, time=1, value=1}]}}})
    Index 0 Param 2=&Add_Float(
     A=&FloatFromCurve(FloatCurve=DataInterface() {propertyValues {Curve {keys [
          {interpMode=RCIM_Cubic, leaveTangent=8.981, tangentMode=RCTM_Break, time=0, value=-0.3}
          {arriveTangent=0.0136, interpMode=RCIM_Cubic, leaveTangent=-5.3647, tangentMode=RCTM_Break, time=0.4833, value=1}
          <at 1 -0.3>
         ]}}})
     B=&Multiply_Float(A=$Particles.BiasedRandom, B=0.5)
    )
    Index 0 Param 3=&RandomRangeFloat(Minimum=0, Maximum=3)
    Index 1 Param 1=&Add_Float(A=$Particles.NormalizedAge, B=&RandomRangeFloat(Minimum=0, Maximum=1))
    Index 1 Param 2=&RandomRangeFloat(Minimum=-50, Maximum=50)
    Index 1 Param 3=3
    Module.Param0WriteEnabled=true
    Module.Param0WriteEnabled1=true
    Module.Param1WriteEnabled=true
    Module.Param1WriteEnabled1=true
    Module.Param2WriteEnabled=true
    Module.Param2WriteEnabled1=true
    Module.Param3WriteEnabled=false
    Module.Param3WriteEnabled1=false
    Write Parameter Index 1=true
   ) {module=DynamicMaterialParameters001}
   &ScaleSpriteSize(
    Uniform Curve Scale=$NPC.NPC_DragonCart_Orb_Rift.RiftScale
    Uniform Curve Sprite Scale=DataInterface() {propertyValues {Curve {keys [
        {arriveTangent=21.5526, interpMode=RCIM_Cubic, time=0, value=0}
        <at 0.1 1 RCIM_Cubic>
        {arriveTangent=-1.8758, interpMode=RCIM_Cubic, leaveTangent=-1.8758, tangentMode=RCTM_User, time=1, value=0}
       ]}}}
   ) {enabled=false}
   &ScaleMeshSize(Scale Factor=&VectorFromFloat(Value=&FloatFromCurve(
      FloatCurve=DataInterface() {propertyValues {Curve {keys [
          {interpMode=RCIM_Cubic, tangentMode=RCTM_Break, time=0, value=0.3}
          {arriveTangent=0.8772, interpMode=RCIM_Cubic, tangentMode=RCTM_Break, time=0.3, value=1}
         ]}}}
      CurveIndex=$System.NormalizedLoopAge
     ))) {module=ScaleMeshSize001}
  ]
  rendererClasses [&NiagaraMeshRendererProperties, &NiagaraSpriteRendererProperties]
  renderers [
   {rendererClass=&NiagaraMeshRendererProperties, rendererIndex=0}
   {rendererClass=&NiagaraSpriteRendererProperties, rendererIndex=1}
  ]
 }
 Plasma {
  emitterUpdateScript [
   &EmitterState()
   &SpawnBurst_Instantaneous(Use Spawn Probability=false, UseLoopCountLimit=false)
   &SpawnRate(
    SpawnRate=&FloatFromCurve(
     FloatCurve=DataInterface() {propertyValues {Curve {keys [
         {arriveTangent=75.6104, interpMode=RCIM_Cubic, leaveTangent=75.6104, tangentMode=RCTM_User, time=0, value=0.1}
         {interpMode=RCIM_Cubic, tangentMode=RCTM_Break, time=0.05, value=3}
         {arriveTangent=-17.3006, interpMode=RCIM_Cubic, tangentMode=RCTM_Break, time=0.3, value=0}
        ]}}}
     CurveIndex=$System.NormalizedLoopAge
     Scale Curve=2
    )
    Use Spawn Probability=false
   )
  ]
  particleSpawnScript [
   /Niagara/Modules/Spawn/Initialization/V2/InitializeParticle(
    Color=<rgba 5 0 0 1>
    Lifetime=&FloatFromCurve(
     FloatCurve=DataInterface() {propertyValues {Curve {keys [
         {interpMode=RCIM_Cubic, tangentMode=RCTM_Break, time=0, value=0.6}
         {arriveTangent=-0.3378, interpMode=RCIM_Cubic, tangentMode=RCTM_Break, time=1, value=0.1}
        ]}}}
     CurveIndex=$System.NormalizedLoopAge
    )
    Mesh Scale=&Multiply_VectorByFloat(Vector=<1 1 1>, Float=50)
    Mesh Scale Mode=::NewEnumerator4
    Position Mode=::NewEnumerator0
    Write Material Random=true
   )
   &InitialMeshOrientation(
    Mesh Axis To Orient=<0 0 0>
    Mesh Orientation Mode=::NewEnumerator2
    Orientation Vector=&RandomRangeVector(Minimum=<-1 0 0>, Maximum=<1 0 0>)
    Use Rotation Vector=false
   )
   &DynamicMaterialParameters(
    Index 0 Param 3=&FloatFromCurve(
     FloatCurve=DataInterface() {propertyValues {Curve {keys [
         {interpMode=RCIM_Cubic, tangentMode=RCTM_Break, time=0, value=-0.3}
         {arriveTangent=-2.8744, interpMode=RCIM_Cubic, tangentMode=RCTM_Break, time=0.6, value=-2}
        ]}}}
     CurveIndex=$System.NormalizedLoopAge
    )
    Module.Index0Float0Write=true
    Module.Index0Float1Write=true
    Module.Index0Float2Write=true
    Module.Index0Float3Write=true
   )
   &CameraOffset(Camera Offset Amount=300)
  ]
  particleUpdateScript [
   &ParticleState()
   &ScaleMeshSize(Scale Factor=&VectorFromFloat(Value=&FloatFromCurve(FloatCurve=DataInterface() {propertyValues {Curve {keys [{interpMode=RCIM_Cubic, tangentMode=RCTM_Break, time=0, value=0.5}, {interpMode=RCIM_Cubic, tangentMode=RCTM_Break, time=1, value=1}]}}})))
   &ScaleMeshSize(Scale Factor=&VectorFromFloat(Value=&FloatFromCurve(
      FloatCurve=DataInterface() {propertyValues {Curve {keys [
          {interpMode=RCIM_Cubic, tangentMode=RCTM_Break, time=0, value=1}
          {arriveTangent=0.8772, interpMode=RCIM_Cubic, tangentMode=RCTM_Break, time=1, value=1.5}
         ]}}}
      CurveIndex=$System.NormalizedLoopAge
     ))) {module=ScaleMeshSize001}
   &StaticMeshLocation(
    MaskSpawnGroup=false
    Mesh Sampling Type:ENiagara_StaticSamplingMode=::NewEnumerator1
    SamplePosOffset=true
    Sampled Position Offset:Vector3f=<15 0 0>
    Socket Sampling Mode:ENiagara_StaticSocketMode=::NewEnumerator2
    Static Mesh:NiagaraDataInterfaceStaticMesh=$User.Static Mesh Reader
   )
   /SetVariables(Particles.Position=&AddVectorToPosition(
     Position=$Particles.Position
     Vector=&TransformVector(Vector=$System.OrbWPO, Source Space=::World, Destination Space=::Local)
    )) {isSetParameters=true}
   &ScaleColor(
    Scale Alpha=&FloatFromCurve(
     FloatCurve=DataInterface() {propertyValues {Curve {
        defaultValue=1
        keys [
         {arriveTangent=0.1462, interpMode=RCIM_Cubic, leaveTangent=0.1462, tangentMode=RCTM_User, time=0, value=0}
         {interpMode=RCIM_Cubic, tangentMode=RCTM_User, time=0.2, value=1}
         {arriveTangent=0.0713, interpMode=RCIM_Cubic, leaveTangent=0.0713, tangentMode=RCTM_User, time=0.7, value=1}
         <at 1 0 RCIM_Cubic>
        ]
       }}}
     Scale Curve=0.3
    )
    ScaleA=true
    ScaleRGB=true
   )
   &ScaleColor(
    Scale Alpha=&FloatFromCurve(
     FloatCurve=DataInterface() {propertyValues {Curve {keys [{interpMode=RCIM_Cubic, tangentMode=RCTM_Break, time=0.2, value=1}, {interpMode=RCIM_Cubic, tangentMode=RCTM_Break, time=0.45, value=0}]}}}
     CurveIndex=$System.NormalizedLoopAge
     Scale Curve=&FNTimeOfDayScalar(Night Scale=0.5, Morning Scale=1, Evening Scale=1)
    )
    ScaleA=true
    ScaleRGB=true
   ) {module=ScaleColor001}
  ]
  rendererClasses [&NiagaraMeshRendererProperties]
  renderers [{rendererClass=&NiagaraMeshRendererProperties, rendererIndex=0}]
 }
 TargetingBeam001 {
  emitterUpdateScript [
   &EmitterState()
   &SpawnBurst_Instantaneous(Use Spawn Probability=false, UseLoopCountLimit=false)
  ]
  particleSpawnScript [
   &InitializeParticle(
    Color=$User.Primary_Color
    Lifetime=$User.Duration
    Mesh Scale Mode=::NewEnumerator3
    Mesh Uniform Scale=10000
    Sprite Size Mode=::NewEnumerator3
    Uniform Sprite Size=2200
    UsePositionOffset=false
    Write Material Random=true
   )
   /SetVariables(Particles.LightExponent=2) {isSetParameters=true}
   /SetVariables(Particles.LightRadius=150) {isSetParameters=true}
   /SetVariables(Particles.LightColor=$User.Light_Color) {isSetParameters=true}
   &DynamicMaterialParameters(
    Index 0 Param 1=2
    Index 0 Param 2=10
    Index 0 Param 3=1
    Index 3 Param 1=10
    Module.Index0Float0Write=true
    Module.Index0Float1Write=true
    Module.Index0Float2Write=true
    Module.Index0Float3Write=true
    Module.Index3Float0Write=true
    Module.Index3Float1Write=true
    Module.Index3Float2Write=true
    Module.Index3Float3Write=true
    Write Parameter Index 3=::NewEnumerator1
   )
   &CameraOffset(Camera Offset Amount=100)
  ]
  particleUpdateScript [
   &ParticleState()
   &StaticMeshLocation(
    MaskSpawnGroup=false
    Mesh Sampling Type:ENiagara_StaticSamplingMode=::NewEnumerator1
    SamplePosOffset=true
    Sampled Position Offset:Vector3f=<10 0 0>
    Socket Sampling Mode:ENiagara_StaticSocketMode=::NewEnumerator2
    Static Mesh:NiagaraDataInterfaceStaticMesh=$User.Static Mesh Reader
   )
   /SetVariables(Particles.Position=&AddVectorToPosition(
     Position=$Particles.Position
     Vector=&TransformVector(Vector=$System.OrbWPO, Source Space=::World, Destination Space=::Local)
    )) {isSetParameters=true}
   &self:ScratchModule_01() {enabled=false, module=ScratchModule_01}
   /SetVariables(
    Particles.OG Dynamic Paramaters 1:Vector4f=&MakeVector4(X=$Particles.Position X, Y=$Particles.Position Y, Z=$Particles.Position Z, W=1)
    Particles.OG Dynamic Parameters 0:Vector4f=&MakeVector4(X=0.1, Y=1, Z=1, W=1)
   ) {enabled=false, isSetParameters=true}
   /SetVariables(Particles.EndPointPosition:NiagaraPosition=&Lerp_Vector(
     A=Linked() {linkedVariable {name=User.InitialPosition, type=NiagaraPosition}}
     B=Linked() {linkedVariable {name=Engine.Owner.Position, type=NiagaraPosition}}
     Alpha=&FloatFromCurve(
      FloatCurve=DataInterface() {propertyValues {Curve {defaultValue=1, keys [<at 0 0 RCIM_Cubic>, {tangentMode=RCTM_User, time=0.4, value=1}]}}}
      CurveIndex=$Particles.Age
     )
    )) {isSetParameters=true}
   /SetVariables(Particles.BeamPosition:NiagaraPosition=&Lerp_Vector(
     A=Linked() {linkedVariable {name=Particles.Position, type=NiagaraPosition}}
     B=Linked() {linkedVariable {name=Particles.EndPointPosition, type=NiagaraPosition}}
     Alpha=0.5
    )) {isSetParameters=true}
   /SetVariables(Particles.BeamSpriteSize:Vector2f=&MakeVector2D(
     X=&FN_LOD_Distance_Remap(Start Distance=2000, End Distance=10000, Minimum Output=10, Maximum Output=50, Clamp Output=true)
     Y=&Multiply_Float(A=&DistanceBetweenPositions(A=$Particles.EndPointPosition, B=$Particles.Position), B=1)
    )) {isSetParameters=true}
   /SetVariables(Particles.BeamSpriteSizeWIder:Vector2f=&MakeVector2D(
     X=&FN_LOD_Distance_Remap(Start Distance=2000, End Distance=10000, Minimum Output=120, Maximum Output=420, Clamp Output=true)
     Y=&Multiply_Float(A=&DistanceBetweenPositions(A=$Particles.EndPointPosition, B=$Particles.Position), B=1)
    )) {isSetParameters=true}
   /SetVariables(Particles.EndPointSpriteSize:Vector2f=&Vector2DFromFloat(Value=&Lerp_Float(
      A=120
      B=130
      Alpha=&FloatFromCurve(
       FloatCurve=DataInterface() {propertyValues {Curve {keys [{interpMode=RCIM_Cubic, tangentMode=RCTM_Break, time=0.5, value=0}, {interpMode=RCIM_Cubic, tangentMode=RCTM_Break, time=1, value=1}]}}}
       CurveIndex=$System.NormalizedLoopAge
      )
     ))) {isSetParameters=true}
   /SetVariables(Particles.EndPointCircleSize:Vector2f=&Vector2DFromFloat(Value=&Lerp_Float(A=10, B=10, Alpha=$System.TargetingLockCurve))) {isSetParameters=true}
   /SetVariables(Particles.WiderBeam_Color:LinearColor=$User.Secondary_Color) {isSetParameters=true}
   &SpriteFacingAndAlignment(
    Sprite Alignment=&Subtract_Position(A=$Particles.Position, B=$Particles.EndPointPosition)
    Sprite Facing=&ReturnMeshOrientationAxis(Mesh Orientation Axis=::XAxis, Mesh Orientation Quat=$Particles.MeshOrientation)
    WriteAlignment=true
    WriteFacing=true
   )
   &ScaleColor(
    Scale RGB=&Multiply_VectorByFloat(
     Vector=<1 1 1>
     Float=&Power(a=&Waveform("Waveform [1]"=::NewEnumerator0, "[1] Frequency"=7, "[1] Amplitude Min/Max"=<0 1>), b=2)
    )
    ScaleA=false
    ScaleRGB=true
   )
   /SetVariables(Particles.OffscreenGlow_Color:LinearColor=&MultiplyLinearColor(
     Linear Color=<rgba 1 1 1 1>
     Scale Factor=&MakeVector4(
      X=100
      Y=100
      Z=100
      W=&FloatFromCurve(FloatCurve=DataInterface() {propertyValues {Curve {keys [<at 0 1>,<at 1 1>]}}})
     )
    )) {enabled=false, isSetParameters=true}
   &ScaleMeshSize(Scale Factor=&VectorFromFloat(Value=&Lerp_Float(A=1, B=0.1, Alpha=$System.TargetingLockCurve))) {enabled=false}
   &DynamicMaterialParameters(
    Index 0 Param 1=1
    Index 0 Param 2=100
    Index 0 Param 3=1
    Module.Index0Float0Write=true
    Module.Index0Float1Write=true
    Module.Index0Float2Write=true
    Module.Index0Float3Write=true
   ) {module=DynamicMaterialParameters001}
  ]
  rendererClasses [
   &NiagaraSpriteRendererProperties
   &NiagaraLightRendererProperties
   &NiagaraMeshRendererProperties
  ]
  renderers [
   {rendererClass=&NiagaraSpriteRendererProperties}
   {rendererClass=&NiagaraSpriteRendererProperties, rendererIndex=1}
   {rendererClass=&NiagaraSpriteRendererProperties, rendererIndex=2}
   {rendererClass=&NiagaraLightRendererProperties, rendererIndex=3}
   {rendererClass=&NiagaraSpriteRendererProperties, rendererIndex=4}
   {rendererClass=&NiagaraSpriteRendererProperties, rendererIndex=5}
   {rendererClass=&NiagaraMeshRendererProperties, rendererIndex=6}
  ]
 }
}
summary {
 systemName=NS_DragonCart_Orb_PlayerWarning_Flashing
 emitters [
  {
   emitterName=TargetingBeam001
   rendererClasses [
    &NiagaraSpriteRendererProperties
    &NiagaraLightRendererProperties
    &NiagaraMeshRendererProperties
   ]
  }
  {emitterName=Glow_02, rendererClasses [&NiagaraSpriteRendererProperties]}
  {emitterName=Plasma, rendererClasses [&NiagaraMeshRendererProperties]}
  {emitterName=Lightning, rendererClasses [&NiagaraMeshRendererProperties, &NiagaraSpriteRendererProperties]}
 ]
 userVariables [
  User.BehindObject {type=NiagaraBool}
  User.BuildUpIntensity {type=NiagaraFloat, defaultValue=1, description=""}
  User.Duration {type=NiagaraFloat, defaultValue=1, description=""}
  User.InitialPosition {type=NiagaraPosition, description="", defaultValue {}}
  User.Light_Color {type=LinearColor, description="", defaultValue {struct=LinearColor, value=<rgba 15 0 0 1>}}
  User.Primary_Color {type=LinearColor, description="", defaultValue {struct=LinearColor, value=<rgba 5 0 0 1>}}
  User.Secondary_Color {type=LinearColor, description="", defaultValue {struct=LinearColor, value=<rgba 1 0.112 0 0.2>}}
  User.Static Mesh Reader {
   type=NiagaraDataInterfaceStaticMesh
   description=""
   defaultValue {
    struct=NiagaraExt_VariableValue_DataInterface
    value {
     dataInterface=&self:NiagaraDataInterfaceStaticMesh_0
     dataInterfaceClass=&NiagaraDataInterfaceStaticMesh
    }
   }
  }
 ]
}
